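# Icon hint widget for the Compete calendar UI, a CommonUserWidget that pairs a small icon with a text label.
# A GridPanel places the centered image in the first column and a gray CommonTextBlock beside it.
# PreConstruct applies the IconMaterial and Name variables to the image and label, so the designer preview reflects their values.
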WidgetBlueprint /Compete/Calendar/IconHints/WBP_IconHint
graphs [EventGraph {nodes [
   Event PreConstruct()
   > SetBrushFromMaterial(
    self=Get Image_Explanation() -> (Image_Explanation:Image Object Reference)
    Material=Get IconMaterial() -> (IconMaterial:Material Interface Object Reference)
   )
   > Set Text(Text=Get Name() -> (Name:Text), self=Get NameExplanation() -> (NameExplanation:Common Text Object Reference))
  ]}]
variables [
 IconMaterial {type=MaterialInterface, default=/FortUILibrary/Textures/MaterialInstances/MI_UIKit_Icon_Add}
 Name {type=string, default="[PH] Text"}
]
widgets {
 info {inheritedWidgetCount=0, parentClass=&CommonUserWidget, rootWidgetClass=&GridPanel, widgetCount=3}
 widgets [
  Root {widgetClassPath=&GridPanel, widget=Root}
  Image_Explanation {
   widgetClassPath=&Image
   bIsVariable=true
   parent=Root
   widget=Image_Explanation
   properties {brush {
     bIsDynamicallyLoaded=false
     drawAs=Image
     imageType=FullColor
     mirroring=NoMirror
     resourceName=None
     resourceObject=/FortUILibrary/Textures/MaterialInstances/MI_UIKit_Icon_Compete
     tiling=NoTile
     imageSize=<32 32>
     margin=<ltrb 0 0 0 0>
     outlineSettings {
      bUseBrushTransparency=false
      roundingType=HalfHeightRadius
      width=0
      color {colorUseRule=UseColor_Specified, specifiedColor=<rgba 0 0 0 0>}
      cornerRadii=<0 0 0 0>
     }
     tintColor {colorUseRule=UseColor_Specified, specifiedColor=<rgba 1 1 1 1>}
     uVRegion {bIsValid=false, max=<0 0>, min=<0 0>}
    }}
   slot {class=GridSlot, properties {horizontalAlignment=HAlign_Center, verticalAlignment=VAlign_Center}}
  }
  NameExplanation {
   widgetClassPath=&CommonTextBlock
   bIsVariable=true
   parent=Root
   widget=NameExplanation
   properties {
    text="[PH] Text"
    colorAndOpacity {colorUseRule=UseColor_Specified, specifiedColor=<rgba 0.5776 0.5776 0.5776 1>}
    font {
     bForceMonospaced=false
     bMaterialIsStencil=false
     fontMaterial=null
     fontObject=&NotoSans
     letterSpacing=0
     monospacedWidth=1
     size=12
     skewAmount=0
     typefaceFontName=NotoSans-Regular
     outlineSettings {
      bApplyOutlineToDropShadows=false
      bMiteredCorners=false
      bSeparateFillAlpha=false
      outlineMaterial=null
      outlineSize=0
      outlineColor=<rgba 0 0 0 1>
     }
    }
    shadowOffset=<0 0>
   }
   slot {class=GridSlot, properties {column=1, verticalAlignment=VAlign_Center, padding=<ltrb 8 0 0 0>}}
  }
 ]
}
